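# Physics asset for the Dragonfruit male one-handed pickaxe, giving its ribbon bones collision bodies and joints for physical motion.
# All bodies are capsules, with paired capsules on the upper ribbon segments and the largest on the master bone.
# Constraints chain the master to a base bone, which forks into a three-bone A strand and a two-bone B strand.
# Every joint shares the same swing limit settings and locks twist.
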PhysicsAsset /BRCosmetics/Weapons/FORT_Melee/Pickaxe_Dragonfruit_Male_1h/Meshes/Dragonfruit_Male_1h_Axe_Physics
bodies {
 dyn_ribbon_a1 [
  {
   shapeType=Capsule
   length=2.47
   radius=0.5
   shapeName=dyn_ribbon_a1_capsule
   center=<1.4722 -0.6851 0.0055>
   rotation=<rot 0.4109 84.7391 -89.2882>
  }
  {
   shapeType=Capsule
   length=2.25
   radius=0.5
   shapeName=dyn_ribbon_a1_capsule
   center=<1.3193 0.5787 -0.0712>
   rotation=<rot 90 -19.4712 -19.4713>
  }
 ]
 dyn_ribbon_a2 [
  {
   shapeType=Capsule
   length=2.3822
   radius=0.5
   shapeName=dyn_ribbon_a2_capsule
   center=<0.9409 -0.8558 0.0019>
   rotation=<rot 0.7185 82.5863 -89.5848>
  }
  {shapeType=Capsule, length=3, radius=0.5, shapeName=dyn_ribbon_a2_capsule, center=<0.9729 0.7772 0.0001>, rotation=<rot 90 35.2644 35.2645>}
 ]
 dyn_ribbon_a3 [{shapeType=Capsule, length=0, radius=0.5, shapeName=dyn_ribbon_a3_capsule, center=<0.0734 -0.0275 0.0045>, rotation=<rot 0 -88.5661 1.3633>}]
 dyn_ribbon_b1 [
  {
   shapeType=Capsule
   length=1.747
   radius=0.5
   shapeName=dyn_ribbon_b1_capsule
   center=<1.4298 -0.5292 -0.0428>
   rotation=<rot -5.9669 -89.1633 -84.9683>
  }
  {shapeType=Capsule, length=2.5, radius=0.5, shapeName=dyn_ribbon_b1_capsule, center=<1.3753 0.5997 0>, rotation=<rot 90 0.0099 0.0099>}
 ]
 dyn_ribbon_b2 [{
   shapeType=Capsule
   length=0
   radius=0.5
   shapeName=dyn_ribbon_b2_capsule
   center=<0.1048 0.1284 -0.2319>
   rotation=<rot -2.2195 -90 -50.3882>
  }]
 dyn_ribbon_base [{
   shapeType=Capsule
   length=0.6715
   radius=1
   shapeName=dyn_ribbon_base_capsule
   center=<0.9806 0.1044 0.2256>
   rotation=<rot -0.6553 -76.0686 -88.8648>
  }]
 dyn_ribbon_mstr [{
   shapeType=Capsule
   length=11.6679
   radius=2
   shapeName=dyn_ribbon_mstr_capsule
   center=<3.2288 -3.8729 0.1158>
   rotation=<rot 2.5258 66.8778 -86.9864>
  }]
}
constraints [
 {bone1Name=dyn_ribbon_base, bone2Name=dyn_ribbon_mstr, swing1LimitDegrees=20, swing2LimitDegrees=20, twistLimitDegrees=5, twistMotion=Locked}
 {bone1Name=dyn_ribbon_a1, bone2Name=dyn_ribbon_base, swing1LimitDegrees=20, swing2LimitDegrees=20, twistLimitDegrees=5, twistMotion=Locked}
 {bone1Name=dyn_ribbon_a2, bone2Name=dyn_ribbon_a1, swing1LimitDegrees=20, swing2LimitDegrees=20, twistLimitDegrees=5, twistMotion=Locked}
 {bone1Name=dyn_ribbon_a3, bone2Name=dyn_ribbon_a2, swing1LimitDegrees=20, swing2LimitDegrees=20, twistLimitDegrees=5, twistMotion=Locked}
 {bone1Name=dyn_ribbon_b1, bone2Name=dyn_ribbon_base, swing1LimitDegrees=20, swing2LimitDegrees=20, twistLimitDegrees=5, twistMotion=Locked}
 {bone1Name=dyn_ribbon_b2, bone2Name=dyn_ribbon_b1, swing1LimitDegrees=20, swing2LimitDegrees=20, twistLimitDegrees=5, twistMotion=Locked}
]
